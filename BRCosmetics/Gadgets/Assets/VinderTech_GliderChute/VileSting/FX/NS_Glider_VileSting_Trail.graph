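# Niagara trail effect for the VileSting glider, with variant colors and separate gameplay and front-end tuning.
# Trails_Gameplay draws two socket-pair anim-trail ribbons, offset over particle life and tinted by time of day.
# Glitchy_Square emits colored sprites from two sockets with curl noise and drag, while Speed_Lines stretches torus-spawned sprites by speed.
# System timelines scale spawn rates, velocities and forces; emitter-age curves delay the ribbon fade-in.
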NiagaraSystem /BRCosmetics/Gadgets/Assets/VinderTech_GliderChute/VileSting/FX/NS_Glider_VileSting_Trail
dependencies {
 usedDataInterfaces [
  &NiagaraDataInterfaceVector2DCurve
  &NiagaraDataInterfaceCurve
  &NiagaraDataInterfaceColorCurve
  &NiagaraDataInterfaceArrayColor
  &NiagaraDataInterfaceSocketReader
 ]
 usedDynamicInputs [
  &RandomRangeVector2D
  &SetBoolByFloatComparison
  &Abs_Float
  &NormalizeVector
 ]
 usedModules [
  &self:SystemSpawnScript.SystemScriptSource.SystemScriptGraph.NiagaraNodeAssignment_1.SetVariables_#1
  &SystemState
  &Timeline
  &FN_Integrator
  &InterpolateOverTime
  &FN_VariantManager
 ]
}
emitters {
 Glitchy_Square {
  emitterUpdateScript [
   &EmitterState(Life Cycle Mode=::NewEnumerator1, Loop Behavior=::NewEnumerator1, Loop Delay=2.4, Loop Duration Mode=::NewEnumerator1, UseLoopDelay=true)
   &SpawnRate(
    SpawnRate=&FN_FrontEndFloatModifier(Gameplay=&Lerp_Float(A=25, B=50, Alpha=$System.Timeline.NormalizedTime), Front End=50)
    Use Spawn Probability=false
   )
  ]
  particleSpawnScript [
   &FN_SetSpawnGroup(Number of Spawn Groups=2)
   &InitializeParticle(
    Color=&VariantColorSelector(
     Variant Amount=0
     Original Color=&SelectLinearColorFromArray(
      Array Sampling Mode=::NewEnumerator0
      Direct Array Index=?
      Color Selection Array=DataInterface() {propertyValues {ColorData [<rgba 2 0 0 1>, <rgba 1.5 0.05 0 1>, <rgba 2 0.3333 0 1>]}}
      Random Seed=0
     )
     Variant 1 Color=&SelectLinearColorFromArray(
      Array Sampling Mode=::NewEnumerator0
      Direct Array Index=?
      Color Selection Array=DataInterface() {propertyValues {ColorData [<rgba 0.1066 0.8167 1.2 1>, <rgba 0.3395 0 2 1>, <rgba 1.0394 0.15 1.5 1>]}}
      Random Seed=0
     )
     Variant 2 Color=?
    )
    Lifetime Max=0.7
    Lifetime Min=0.3
    Lifetime Mode=::NewEnumerator1
    Sprite Size Mode=::NewEnumerator3
    Uniform Sprite Size=&Multiply_Float(A=&RandomRangeFloat(Minimum=5, Maximum=7), B=1)
    UsePositionOffset=true
    Write Material Random=true
   )
   &SocketLocation(
    SamplePosOffset=true
    Sampled Position Offset=&FN_FrontEndVector3Modifier(
     Gameplay=&Multiply_VectorByFloat(
      Vector=&MaskVector(Mask Source=0, Mask Parameter=?, Mask=0, Masked Vector=<1 0.3 0>, Passthrough Vector=<-1 -0.3 0>)
      Float=&RandomRangeFloat(Minimum=0, Maximum=145)
     )
     Front End=&Multiply_VectorByFloat(
      Vector=&MaskVector(Mask Source=0, Mask Parameter=?, Mask=0, Masked Vector=<1 0.3 0>, Passthrough Vector=<-1 -0.3 0>)
      Float=&RandomRangeFloat(Minimum=0, Maximum=145)
     )
    )
    Socket Index=&MaskIntBySpawnGroup(Masked Int=0, Passthrough Int=2, Spawn Group Mask=0)
    Socket Reader=$System.Socket Reader
    Socket Sampling Mode=::NewEnumerator3
    VelocitySpeedLimit=false
   )
   &AddVelocity(
    Cone Angle=&FN_FrontEndFloatModifier(Gameplay=&Lerp_Float(A=15, B=20, Alpha=$System.Timeline.NormalizedTime), Front End=20)
    Module.UseConeFalloff=false
    Module.UseVelDistribution=false
    Override Seed=false
    Velocity Mode=::NewEnumerator2
    Velocity Speed=&FN_FrontEndFloatModifier(
     Gameplay=&Lerp_Float(
      A=0
      B=&Multiply_Float(
       A=&RandomRangeFloat(Minimum=300, Maximum=600)
       B=&Lerp_Float(A=0.5, B=0.8, Alpha=$System.Timeline.NormalizedTime)
      )
      Alpha=&FloatFromCurve(
       FloatCurve=DataInterface() {propertyValues {Curve {
          defaultValue=1
          keys [{interpMode=RCIM_Cubic, tangentMode=RCTM_User, time=1, value=0}, {interpMode=RCIM_Cubic, tangentMode=RCTM_User, time=2, value=1}]
         }}}
       CurveIndex=$Emitter.Age
      )
     )
     Front End=&RandomRangeFloat(Minimum=-200, Maximum=-400)
    )
   )
   &CameraOffset(Camera Offset Amount=5)
  ]
  particleUpdateScript [
   &ParticleState()
   &LinearForce(Force=&Multiply_VectorByFloat(
     Vector=&Lerp_Vector(
      A=<0 0 0>
      B=<0 1 0>
      Alpha=&Multiply_Float(A=$System.Timeline001.Time, B=$System.InterpolateOverTime001.InterpolateOverTime_Float.Moving Average)
     )
     Float=&FloatFromCurve(
      FloatCurve=DataInterface() {propertyValues {Curve {
         defaultValue=1
         keys [
          {arriveTangent=-0.0013, interpMode=RCIM_Cubic, leaveTangent=-0.0013, tangentMode=RCTM_User, time=0, value=0}
          {arriveTangent=1.3901, interpMode=RCIM_Cubic, leaveTangent=1.3901, tangentMode=RCTM_User, time=1, value=1}
         ]
        }}}
      Scale Curve=&Multiply_Float(A=300, B=&Lerp_Float(A=1, B=1.8, Alpha=$System.Timeline.NormalizedTime))
     )
    ))
   &Drag(Drag=3, Use Linear Drag=true, Use Rotational Drag=false)
   &ScaleSpriteSize(Scale Factor=&Vector2DFromFloat(Value=&FloatFromCurve(FloatCurve=DataInterface() {propertyValues {Curve {
         defaultValue=1
         keys [{interpMode=RCIM_Cubic, tangentMode=RCTM_User, time=0, value=0.5}, {interpMode=RCIM_Cubic, tangentMode=RCTM_User, time=0.7, value=1}]
        }}}))) {module=ScaleSpriteSize001}
   &ScaleColor(
    Scale Alpha=&FloatFromCurve(FloatCurve=DataInterface() {propertyValues {Curve {
        defaultValue=0
        keys [
         {arriveTangent=4.2091, interpMode=RCIM_Cubic, leaveTangent=4.2091, tangentMode=RCTM_User, time=0, value=0}
         {interpMode=RCIM_Cubic, tangentMode=RCTM_User, time=0.5, value=1}
         {arriveTangent=-3.8223, tangentMode=RCTM_User, time=1, value=0}
        ]
       }}})
    Scale RGB=&VectorFromFloat(Value=&FloatFromCurve(FloatCurve=DataInterface() {propertyValues {Curve {
         defaultValue=0
         keys [{interpMode=RCIM_Cubic, tangentMode=RCTM_User, time=0, value=1}, {interpMode=RCIM_Cubic, tangentMode=RCTM_User, time=1, value=0}]
        }}}))
    ScaleA=true
    ScaleRGB=true
   ) {module=ScaleColor001}
   &CurlNoiseForce(
    Bias Noise=false
    Noise Strength=&Multiply_Float(A=500, B=&Lerp_Float(A=0.3, B=1, Alpha=$System.Timeline.NormalizedTime))
    Pan Noise=false
   )
   &SolveForcesAndVelocity(Clamp Velocity=false, Limit Acceleration=false)
   &SpriteFacingAndAlignment(Sprite Facing=$Particles.ShapeLocation.ShapeTangent, WriteAlignment=false, WriteFacing=true)
   &FN_Sprite_Attributes(
    Color=&MakeLinearColorFromVectorAndFloat(
     "Vector (RGB)"=&MakeVectorFromLinearColorRGB(Color=&VariantColorSelector(
       Variant Amount=0
       Original Color=&Multiply_LinearColorByFloat(Linear Color=$Particles.Color, Float=1)
       Variant 1 Color=&Multiply_LinearColorByFloat(Linear Color=$Particles.Color, Float=0.5)
       Variant 2 Color=?
      ))
     "Float (Alpha)"=&FloatFromCurve(FloatCurve=DataInterface() {propertyValues {Curve {
         defaultValue=1
         keys [
          {arriveTangent=3.6295, interpMode=RCIM_Cubic, leaveTangent=3.6295, tangentMode=RCTM_User, time=0, value=0}
          {interpMode=RCIM_Cubic, tangentMode=RCTM_User, time=0.5, value=1}
          {tangentMode=RCTM_User, time=1, value=0}
         ]
        }}})
    )
    Create Color Attribute=true
    Create SpriteSize Attribute=true
    SpriteSize=&Multiply_Vector2D_ByFloat(Float=1.5, Vector2D=$Particles.SpriteSize)
    Write to Module Namespace=true
   )
  ]
  rendererClasses [&NiagaraSpriteRendererProperties]
  renderers [
   {rendererClass=&NiagaraSpriteRendererProperties}
   {rendererClass=&NiagaraSpriteRendererProperties, rendererIndex=1}
  ]
 }
 Speed_Lines {
  emitterUpdateScript [
   &EmitterState()
   &SpawnRate(
    SpawnRate=&FN_FrontEndFloatModifier(Gameplay=&Lerp_Float(A=0, B=90, Alpha=$System.Timeline.NormalizedTime), Front End=50)
    Use Spawn Probability=true
   ) {module=SpawnRate001}
  ]
  particleSpawnScript [
   &InitializeParticle(
    Color=&MakeLinearColorFromVectorAndFloat(
     "Vector (RGB)"=<1 1 1>
     "Float (Alpha)"=&Multiply_Float(
      A=&UniformRangedFloat(Minimum=0.1, Maximum=0.15)
      B=&FN_FrontEndFloatModifier(Gameplay=0.5, Front End=0.5)
     )
    )
    Lifetime Max=0.16
    Lifetime Min=0.1
    Lifetime Mode=::NewEnumerator1
    Sprite Size Max=<3 10>
    Sprite Size Min=<2 2>
    Sprite Size Mode=::NewEnumerator2
    UsePositionOffset=false
    Write Material Random=true
   )
   &TorusLocation(
    Align Torus With Axis=true
    Handle Radius=0
    Large Radius:NiagaraFloat=&FN_FrontEndFloatModifier(
     Gameplay=&RandomRangeFloat(Minimum=150, Maximum=200)
     Front End=&RandomRangeFloat(Minimum=100, Maximum=200)
    )
    MaskSpawnGroup=false
    Module.Override Randomness=false
    Torus Axis=<1 0 0>
    Torus Origin:NiagaraPosition=&AddVectorToPosition(
     Position=$Particles.Position
     Vector=&FN_FrontEndVector3Modifier(Gameplay=<0 0 0>, Front End=<200 0 -40>)
    )
   )
   &AddVelocityFromPoint(
    Module.Use Falloff Distance=false
    Origin Offset=<-500 0 0>
    Velocity Strength=&Multiply_Float(
     A=&FN_FrontEndFloatModifier(Gameplay=600, Front End=0)
     B=&MakeCustomFloatFromBool(Bool=$User.bIsFrontEndPreview, True Float=1, False Float=1)
    )
   )
   &AddVelocity(Velocity=&Multiply_Vector(
     A=&RandomRangeVector(Minimum=<300 0 0>, Maximum=<700 0 0>)
     B=&MakeVector(X=&MakeCustomFloatFromBool(Bool=$User.bIsFrontEndPreview, True Float=-1, False Float=1), Y=1, Z=1)
    ))
  ]
  particleUpdateScript [
   &SolveForcesAndVelocity(Clamp Velocity=false, Limit Acceleration=false)
   &ParticleState()
   &ScaleSpriteSize(Scale Factor=&Vector2DFromCurve(
     Vector2Curve=DataInterface() {propertyValues {XCurve {keys [<at 0 0.5 RCIM_Cubic>,<at 0.5 1 RCIM_Cubic>]}, YCurve {keys [<at 0 0 RCIM_Cubic>,<at 0.5 1 RCIM_Cubic>]}}}
     ScaleCurve=&Vector2DFromFloat(Value=1)
    ))
   &ScaleSpriteSizeBySpeed(Max Scale Factor=<1 50>, Min Scale Factor=<1 40>, Velocity Threshold=100)
   &ScaleColor(
    Scale Alpha=&FN_FrontEndFloatModifier(
     Gameplay=&Lerp_Float(A=0, B=1, Alpha=$System.Timeline.NormalizedTime)
     Front End=&FloatFromCurve(
      FloatCurve=DataInterface() {propertyValues {Curve {
         defaultValue=1
         keys [{interpMode=RCIM_Cubic, tangentMode=RCTM_User, time=2.3, value=0}, {interpMode=RCIM_Cubic, tangentMode=RCTM_User, time=2.8, value=1}]
        }}}
      CurveIndex=$Emitter.Age
      Scale Curve=0.5
     )
    )
    Scale RGB=&VectorFromFloat(Value=1)
    ScaleA=true
    ScaleRGB=true
   )
  ]
  rendererClasses [&NiagaraSpriteRendererProperties]
  renderers [{rendererClass=&NiagaraSpriteRendererProperties}]
 }
 Trails_Gameplay {
  emitterSpawnScript [&FN_RibbonIDOffsetUtility()]
  emitterUpdateScript [
   &EmitterState(Life Cycle Mode=::NewEnumerator1, Loop Behavior=::NewEnumerator1, Loop Duration Mode=::NewEnumerator1, UseLoopDelay=false)
   &FN_SpawnAnimTrail(Number of Anim Trails=2)
  ]
  particleSpawnScript [
   &FN_InitializeAnimTrailBoneIndecies(First Bone or Socket Index=1, Second Bone or Socket Index=0)
   &FN_InitializeAnimTrailBoneIndecies(AnimTrail Selection=1, First Bone or Socket Index=3, Second Bone or Socket Index=2) {module=FN_InitializeAnimTrailBoneIndecies001}
   &FN_InitializeAnimTrail(
    Color=&VariantColorSelector(Variant Amount=0, Original Color=<rgba 1 1 1 1>, Variant 1 Color=<rgba 1 1 1 1>, Variant 2 Color=?)
    Lifetime=&MakeCustomFloatFromBool(Bool=$Particles.Locked, True Float=0, False Float=0.3)
    Position Offset=&FN_FrontEndVector3Modifier(Gameplay=<0 0 0>, Front End=<0 0 0>)
   )
  ]
  particleUpdateScript [
   &FN_UpdateAnimTrail(Locked Particle RibbonWidth=1) {module=FN_UpdateAnimTrail001}
   &SocketLocation(
    Orientation Sampling=::NewEnumerator1
    Position Sampling=::NewEnumerator1
    SamplePosOffset=false
    Sampled Velocity Coordinate Space:ENiagaraCoordinateSpace=::Simulation
    Socket Index=1
    Socket Reader=$System.Socket Reader
    Socket Sampling Mode=::NewEnumerator3
    VelocitySpeedLimit=false
   )
   &SocketLocation(
    Orientation Sampling=::NewEnumerator1
    Position Sampling=::NewEnumerator1
    SamplePosOffset=false
    Sampled Velocity Coordinate Space:ENiagaraCoordinateSpace=::Simulation
    Socket Index=3
    Socket Reader=$System.Socket Reader
    Socket Sampling Mode=::NewEnumerator3
    VelocitySpeedLimit=false
   ) {module=SocketLocation001}
   /SetVariables(Particles.Position=&Add_Vector(
     A=Linked() {linkedVariable {name=Particles.RibbonPosFromCentre, type=NiagaraPosition}}
     B=&Multiply_VectorByFloat(
      Vector=/Niagara/DynamicInputs/Vectors/CrossProduct(
       Vector A:Vector3f=&RotateVectorByQuaternion(
        Vector To Rotate=&MaskVector(Mask Source=0, Mask Parameter=?, Mask=0, Masked Vector=<0 0 1>, Passthrough Vector=<0 0 -1>)
        Quaternion=/Niagara/DynamicInputs/Helpers/MaskQuaternion(
         Mask Source:NiagaraInt32=0
         Mask Parameter:NiagaraInt32=?
         Mask:NiagaraInt32=0
         Masked Quaternion:Quat4f=$Output.SocketLocation.SampledRotation
         Passthrough Quaternion:Quat4f=$Output.SocketLocation001.SampledRotation
        )
       )
       Vector B:Vector3f=&TransformVector(Vector=<0 -1 0>, Destination Space=::Simulation)
       Normalize Incoming Vectors:NiagaraBool=true
      )
      Float=&Multiply_Float(
       A=&FloatFromCurve(FloatCurve=DataInterface() {propertyValues {Curve {defaultValue=1, keys [{tangentMode=RCTM_User, time=0, value=0}, {tangentMode=RCTM_User, time=1, value=1}]}}})
       B=&FN_FrontEndFloatModifier(
        Gameplay=&Lerp_Float(
         A=0
         B=&Lerp_Float(A=150, B=300, Alpha=$System.Timeline.Time)
         Alpha=&FloatFromCurve(
          FloatCurve=DataInterface() {propertyValues {Curve {
             defaultValue=1
             keys [{interpMode=RCIM_Cubic, tangentMode=RCTM_User, time=2.3, value=0}, {interpMode=RCIM_Cubic, tangentMode=RCTM_User, time=3, value=1}]
            }}}
          CurveIndex=$Emitter.Age
         )
        )
        Front End=&Lerp_Float(
         A=0
         B=-150
         Alpha=&FloatFromCurve(
          FloatCurve=DataInterface() {propertyValues {Curve {
             defaultValue=1
             keys [{interpMode=RCIM_Cubic, tangentMode=RCTM_User, time=2.3, value=0}, {interpMode=RCIM_Cubic, tangentMode=RCTM_User, time=3, value=1}]
            }}}
          CurveIndex=$Emitter.Age
         )
        )
       )
      )
     )
    )) {isSetParameters=true}
   /SetVariables(Particles.Position=&Divide_Vector(A:Vector3f=Linked() {linkedVariable {name=Particles.Position, type=NiagaraPosition}}, B:Vector3f=$Engine.Owner.Scale)) {isSetParameters=true}
   /SetVariables(Particles.RibbonFacing=&Subtract_Vector(
     A=Linked() {linkedVariable {name=Particles.RibbonPosFromSecond, type=NiagaraPosition}}
     B=Linked() {linkedVariable {name=Particles.RibbonPosFromFirst, type=NiagaraPosition}}
    )) {isSetParameters=true}
   &ScaleColor(
    Scale Alpha=&Multiply_Float(
     A=&FN_FrontEndFloatModifier(Gameplay=1, Front End=1)
     B=&FloatFromCurve(
      FloatCurve=DataInterface() {propertyValues {Curve {
         defaultValue=1
         keys [{interpMode=RCIM_Cubic, tangentMode=RCTM_User, time=2.4, value=0}, {interpMode=RCIM_Cubic, tangentMode=RCTM_User, time=3, value=1}]
        }}}
      CurveIndex=$Emitter.Age
      Scale Curve=/BRCosmetics/Weapons/FORT_Melee/Pickaxe_GateHound/FX/VariantManager/VariantFloatSelector(
       Original Float:NiagaraFloat=&FNTimeOfDayScalar(Night Scale=0.8, Morning Scale=0.8, Day Scale=1.1, Evening Scale=0.8)
       Variant 1 Float:NiagaraFloat=0
       Variant 2 Float:NiagaraFloat=?
       Variant 3 Float:NiagaraFloat=?
       Variant Amount:NiagaraInt32=0
      )
     )
    )
    Scale RGB=&MakeVectorFromLinearColorRGB(Color=&FNTimeOfDayColor(
      Night Color=<rgba 0.5625 0.6225 0.75 1>
      Morning Color=<rgba 0.7812 0.5615 0.5615 1>
      Day Color=<rgba 1 1 1 1>
      Evening Color=<rgba 0.7812 0.5615 0.5615 1>
     ))
    ScaleA=true
    ScaleRGB=true
   )
   &ScaleRibbonWidth(Ribbon Width Scale=&FloatFromCurve(FloatCurve=DataInterface() {propertyValues {Curve {defaultValue=0, keys [<at 0 1>,<at 1 1.2>]}}}))
   &DynamicMaterialParameters(
    Index 0 Param 2=0.4
    Module.Index0Float0Write=true
    Module.Index0Float1Write=true
    Module.Index0Float2Write=true
    Module.Index0Float3Write=true
   )
   /SetVariables(Particles.SnapVariantColor:LinearColor=&MakeLinearColorFromVectorAndFloat(
     "Vector (RGB)"=&MakeVectorFromLinearColorRGB(Color=&FNTimeOfDayColor(
       Night Color=<rgba 0.5625 0.6225 0.75 1>
       Morning Color=<rgba 0.7812 0.5615 0.5615 1>
       Day Color=<rgba 1 1 1 1>
       Evening Color=<rgba 0.7812 0.5615 0.5615 1>
      ))
     "Float (Alpha)"=&Multiply_Float(
      A=&FN_FrontEndFloatModifier(Gameplay=1, Front End=1)
      B=&FloatFromCurve(
       FloatCurve=DataInterface() {propertyValues {Curve {
          defaultValue=1
          keys [{interpMode=RCIM_Cubic, tangentMode=RCTM_User, time=2.4, value=0}, {interpMode=RCIM_Cubic, tangentMode=RCTM_User, time=3, value=1}]
         }}}
       CurveIndex=$Emitter.Age
       Scale Curve=&VariantFloatSelector(
        Original Float:NiagaraFloat=0
        Variant 1 Float:NiagaraFloat=&FNTimeOfDayScalar(Night Scale=0.6, Morning Scale=0.7, Day Scale=0.8, Evening Scale=0.7)
        Variant 2 Float:NiagaraFloat=?
        Variant 3 Float:NiagaraFloat=?
        Variant Amount:NiagaraInt32=0
       )
      )
     )
    )) {isSetParameters=true}
   /SetVariables(Particles.TipColor:LinearColor=&MakeLinearColorFromVectorAndFloat(
     "Vector (RGB)"=&MakeVectorFromLinearColorRGB(Color=&MultiplyLinearColor(
       Linear Color=&VariantColorSelector(Variant Amount=0, Original Color=<rgba 4 2.2 0.4 1>, Variant 1 Color=<rgba 10 0 0 1>, Variant 2 Color=?)
       Scale Factor=&MakeVector4_FromLinearColor(LinearColor=&FNTimeOfDayColor(
         Night Color=<rgba 0.5625 0.6225 0.75 1>
         Morning Color=<rgba 0.7812 0.5615 0.5615 1>
         Day Color=<rgba 1 1 1 1>
         Evening Color=<rgba 0.7812 0.5615 0.5615 1>
        ))
      ))
     "Float (Alpha)"=&Multiply_Float(
      A=&FloatFromCurve(
       FloatCurve=DataInterface() {propertyValues {Curve {
          defaultValue=1
          keys [{interpMode=RCIM_Cubic, tangentMode=RCTM_User, time=0, value=1}, {interpMode=RCIM_Cubic, tangentMode=RCTM_User, time=0.3, value=0}]
         }}}
       Scale Curve=&FN_FrontEndFloatModifier(Gameplay=1.5, Front End=1)
      )
      B=&FloatFromCurve(
       FloatCurve=DataInterface() {propertyValues {Curve {
          defaultValue=1
          keys [{interpMode=RCIM_Cubic, tangentMode=RCTM_User, time=2.6, value=0}, {interpMode=RCIM_Cubic, tangentMode=RCTM_User, time=3, value=1}]
         }}}
       CurveIndex=$Emitter.Age
       Scale Curve=&VariantFloatSelector(
        Original Float:NiagaraFloat=0.3
        Variant 1 Float:NiagaraFloat=0.3
        Variant 2 Float:NiagaraFloat=?
        Variant 3 Float:NiagaraFloat=?
        Variant Amount:NiagaraInt32=0
       )
      )
     )
    )) {isSetParameters=true}
  ]
  rendererClasses [&NiagaraRibbonRendererProperties]
  renderers [
   {rendererClass=&NiagaraRibbonRendererProperties}
   {rendererClass=&NiagaraRibbonRendererProperties, rendererIndex=1}
   {rendererClass=&NiagaraRibbonRendererProperties, rendererIndex=2}
  ]
 }
}
summary {
 systemName=NS_Glider_VileSting_Trail
 emitters [
  {emitterName=Speed_Lines, rendererClasses [&NiagaraSpriteRendererProperties]}
  {emitterName=Trails_Gameplay, rendererClasses [&NiagaraRibbonRendererProperties]}
  {emitterName=Glitchy_Square, rendererClasses [&NiagaraSpriteRendererProperties]}
 ]
 userVariables [
  User.bIsFrontEnd {type=NiagaraBool}
  User.bIsFrontEndPreview {type=NiagaraBool}
  User.ForwardDot {type=NiagaraFloat, defaultValue=0, description=""}
  User.RibbonIDOffset {type=NiagaraInt32}
  User.RightDot {type=NiagaraFloat, defaultValue=0, description=""}
  User.Variant {type=NiagaraFloat, defaultValue=0, description=""}
 ]
}
